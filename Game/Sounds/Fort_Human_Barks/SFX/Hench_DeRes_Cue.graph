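# Sound cue for a henchman de-res effect, filed with the human bark SFX.
# A single Random node picks one of three equally weighted de-res wave variations.
# It borrows the gadget-activate attenuation, plays through the player foley sound class, and carries a slight volume boost.
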SoundCue /Game/Sounds/Fort_Human_Barks/SFX/Hench_DeRes_Cue
root=Random_0
nodes [
 {id=Random_0, children [WavePlayer_1,WavePlayer_2,WavePlayer_3], properties {weights [1,1,1]}}
 {id=WavePlayer_1, properties {soundWaveAssetPtr=&/Hench_DeRes_01}}
 {id=WavePlayer_2, properties {soundWaveAssetPtr=&/Hench_DeRes_02}}
 {id=WavePlayer_3, properties {soundWaveAssetPtr=&/Hench_DeRes_03}}
]
properties {
 attenuationSettings=&Fort_GadgetActivate_Att
 maxDistance=4000
 soundClassObject=&Fort_Player_Foley
 volumeMultiplier=1.2
}
