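# Sound cue for the interrogation "indicated" ping, routed through the critical in-game UI sound class.
# A short fixed delay node feeds a single wave player holding the Interrogate ping wave, at reduced cue volume.
# Attenuation override turns off distance falloff and enables listener focus.
# Concurrency override caps instances per owner, stopping the oldest and applying volume scaling to overlapping pings.
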SoundCue /Game/Sounds/Fort_Human_Barks/SFX/Mang_Interrogate_Indicated_Cue
root=Delay_0
nodes [
 {id=Delay_0, children [WavePlayer_3], properties {delayMax=0.25, delayMin=0.25}}
 {id=WavePlayer_3, properties {soundWaveAssetPtr=/Game/Athena/Sounds/Character/Interrogate/Interrogate_Ping_01}}
]
properties {
 bOverrideAttenuation=true
 bOverrideConcurrency=true
 maxDistance=2097152
 soundClassObject=&Fort_UI_InGame_Critical
 volumeMultiplier=0.5
 attenuationOverrides {bAttenuate=false, bEnableListenerFocus=true, focusVolumeAttenuation=1.3, stereoSpread=512}
 concurrencyOverrides {bLimitToOwner=true, maxCount=3, resolutionRule=StopOldest, volumeScale=0.75, volumeScaleAttackTime=0.2}
}
