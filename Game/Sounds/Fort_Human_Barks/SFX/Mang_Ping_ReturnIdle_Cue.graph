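# Sound cue named for the MANG henchman ping's return to idle in the human barks SFX set.
# Its graph is empty, with no root or wave nodes, so it has no audio to play.
# It still carries the henchman ping attenuation, the critical in-game UI sound class and a ping concurrency set.
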SoundCue /Game/Sounds/Fort_Human_Barks/SFX/Mang_Ping_ReturnIdle_Cue
root=null
nodes []
properties {
 attenuationSettings=/Game/Sounds/Fort_Human_Barks/MANG/Henchman_MANG_Ping_Att
 maxDistance=4000
 soundClassObject=&Fort_UI_InGame_Critical
 volumeMultiplier=1
 concurrencySet [&/Mang_Ping_Conc]
}
